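; GA_Quest_Swim is a Battle Pass S18 quest gameplay ability that reports a swim stat when activated from a gameplay event.
; On activation it calls the parent handler, then sends a complex custom stat event with a count of one, tagged by QuestStatEventTags, through the owning player's quest manager.
; A debug print gated by bDebugAbility logs the call, and the ability then ends immediately.
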(event Ability|EventActivateAbilityFromEvent
  (params (EventData :type Gameplay%20Event%20Data%20Structure%20%28by%20ref%29))
  (|CallParentFunction EventData)
  (Development|PrintString "@@@ ActivateAbilityFromEvent [GA_Quest_Swim]: CALLED - Sending Go Swim Custom Stat Event!" false (Variables|Debug|GetbDebugAbility))
  (Quest|SendComplexCustomStatEvent (Variables|GAVars|Data|GetOwningPlayerQuestManager) 0 (Variables|Default|GetQuestStatEventTags) "(GameplayTags=)" 1)
  (Ability|EndAbility))
